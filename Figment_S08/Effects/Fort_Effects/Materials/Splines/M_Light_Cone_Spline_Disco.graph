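# Additive, two-sided mesh-particle material for disco-style light cone beams with color-cycling hues.
# Vertex blue plus a particle Random value is quantized to quarters, offsetting UVs into the light-shaft texture and seeding a time-driven HueShift.
# Particle color tints the sum of the texture's green channel and the shifted RGB.
# SplineThicken sets width from WidthMult, while particle alpha multiplies a depth fade, bypassed on ES2/ES3.1.
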Material /Figment_S08/Effects/Fort_Effects/Materials/Splines/M_Light_Cone_Spline_Disco
graph {
 properties {
  bUsedWithMeshParticles=true
  blendMode=BLEND_Additive
  pixelDepthOffsetMode=PDOM_Legacy
  previewMesh=/Game/Effects/Fort_Effects/Mesh/Splines/SM_Spline_Burst
  twoSided=true
  usedShadingModels=Default Lit
 }
 nodes [
  n1 = ParticleColor()
  n2 = Constant2Vector() {r=1}
  n3 = DynamicParameter() {paramNames [WidthMult,Random,ChangeSpeed,Param4]}
  n4 = Add(A=VertexColor().B, B=@n3.Random)
  n5 = Divide(A=Round(None=Multiply(A=@n4) {constB=4})) {constB=4}
  n6 = TextureSample(UVs=Add(A=TextureCoordinate() {uTiling=0.25}, B=Multiply(A=@n2, B=@n5))) {texture=/Game/Effects/Fort_Effects/Textures/Mask/T_Flare_LightShafts}
  n7 = Add(A=@n5, B=Multiply(A=Time(), B=@n3.ChangeSpeed))
  Add(
   A=Multiply(A=@n1.RGB, B=@n6.G)
   B=Multiply(
    A=@n1.RGB
    B=MaterialFunctionCall(Hue Shift Percentage=@n7, Texture=@n6.RGB) {materialFunction=&HueShift}.Result
   )
  )
  n8 = Multiply(A=@n3.WidthMult, B=ScalarParameter() {defaultValue=20, parameterName=BaseWidth})
  MaterialFunctionCall(WidthBase=@n8, WidthTip=@n8) {materialFunction=&SplineThicken}
  n9 = DepthFade(FadeDistance=ScalarParameter() {defaultValue=50, parameterName=DepthFadeDist}) {fadeDistanceDefault=25}
  n10 = Constant() {r=1}
  Multiply(A=@n1.A, B=FeatureLevelSwitch(Default=@n9, ES2=@n10, ES3_1=@n10, SM4_REMOVED=@n9))
 ]
}
